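# MF_PerceivedBrightness is a material function that converts an RGB color, described as linear-encoded, into a scalar grayscale brightness value.
# It squares each channel, weights the results with Rec. 601 luma coefficients, sums them through AddComponents and takes the square root.
# The result is a weighted root-mean-square, as in the HSP perceived brightness formula, rather than a plain weighted sum.
# The input defaults to white when unconnected.
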MaterialFunction /Emporium/MaterialFunctions/MF_PerceivedBrightness
graph {nodes [
  n1 = FunctionInput(Preview=Constant3Vector() {constant=<rgba 1 1 1 0>}) {bUsePreviewValueAsDefault=true, description=Linear-encoded color, inputName=RGBColor}
  n2 = Multiply(A=Constant3Vector() {constant=<rgba 0.299 0.587 0.114 0>}, B=Multiply(A=@n1, B=@n1))
  FunctionOutput(None=SquareRoot(None=MaterialFunctionCall(f3=@n2) {materialFunction=&AddComponents}.f3)) {outputName="Grayscale (S)"}
 ]}
