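# Material function that rotates a tangent-space normal around its Z axis, leaving Z unchanged.
# The SinCos input supplies sine in R and cosine in G, so the function computes no trigonometry itself.
# These build rotated X and Y basis vectors, which Transform3x3Matrix applies to NormalsTS with a fixed Z basis.
# Both inputs use their previews as defaults, giving an unrotated flat normal when unconnected.
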MaterialFunction /Emporium/MaterialFunctions/MF_RotateNormals_TS
graph {nodes [
  n1 = Constant3Vector() {constant=<rgba 0 0 1 0>}
  n2 = FunctionInput(Preview=Constant2Vector() {g=1}) {bUsePreviewValueAsDefault=true, inputName=SinCos, inputType=FunctionInput_Vector2, sortPriority=2}
  n3 = ComponentMask(None=@n2) {g=true}
  n4 = ComponentMask(None=@n2) {r=true}
  n5 = Multiply(A=@n4) {constB=-1}
  FunctionOutput(None=MaterialFunctionCall(
    VectorToTransform=FunctionInput(Preview=@n1) {bUsePreviewValueAsDefault=true, inputName=NormalsTS}
    BasisX=MaterialFunctionCall(X=@n3, Y=@n5) {materialFunction=&MakeFloat3}.Result
    BasisY=MaterialFunctionCall(X=@n4, Y=@n3) {materialFunction=&MakeFloat3}.Result
    BasisZ=Constant3Vector() {constant=<rgba 0 0 1 0>}
   ) {materialFunction=&Transform3x3Matrix}.Transformed Vector)
 ]}
