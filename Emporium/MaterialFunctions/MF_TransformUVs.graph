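# Material function that applies a scale, rotation and offset transform to the default texture coordinates.
# OffsetAndScale packs offset in RG and scale in BA: UVs are multiplied by the scale, rotated around the origin by MF_RotateV2, then shifted by the offset.
# Rotation arrives as a precomputed SinCosAngle pair, and both inputs default to an identity transform.
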MaterialFunction /Emporium/MaterialFunctions/MF_TransformUVs
graph {nodes [
  n1 = FunctionInput(Preview=Constant4Vector() {constant=<rgba 0 0 1 1>}.RGBA) {bUsePreviewValueAsDefault=true, inputName=OffsetAndScale, inputType=FunctionInput_Vector4, previewValue=<0 0 1 1>}
  n2 = Multiply(A=TextureCoordinate(), B=ComponentMask(None=@n1) {a=true, b=true})
  n3 = Constant2Vector()
  n4 = FunctionInput(Preview=Constant2Vector() {g=1}) {bUsePreviewValueAsDefault=true, inputName=SinCosAngle, inputType=FunctionInput_Vector2, previewValue=<0 1 0 1>}
  FunctionOutput(None=Add(
    A=MaterialFunctionCall!(UV=@n2, RotationCenter=@n3, SinCosAngle=@n4) {materialFunction=&/MF_RotateV2}.Result
    B=ComponentMask(None=@n1) {g=true, r=true}
   ))
 ]}
